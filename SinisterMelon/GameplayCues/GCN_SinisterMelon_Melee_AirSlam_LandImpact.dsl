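; Burst gameplay cue for the SinisterMelon melee air slam landing, which positions the impact VFX on the surface beneath the target.
; On burst, it line-traces by object type straight down from the target, ignoring it, then moves the first particle component to the hit location, nudged along the target's forward vector.
; It also sets that effect's IsLocalPlayer parameter from whether the target pawn is locally controlled.
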(fn OnBurst(EitherVFX-preferthisversion)
  (params (MyTarget :type Actor%20Object%20Reference) (Parameters :type Gameplay%20Cue%20Parameters%20Structure%20%28by%20ref%29) (ParticleComponents :type Array%20of%20FXSystem%20Component%20Object%20References) (AudioComponents :type Array%20of%20Audio%20Component%20Object%20References) (BurstCameraShakeInstance :type Camera%20Shake%20Base%20Object%20Reference) (BurstDecalInstance :type Decal%20Actor%20Object%20Reference))
  (bind _outhit (Collision|LineTraceForObjects 0 (Transformation|GetActorLocation MyTarget) (- (Transformation|GetActorLocation MyTarget) "0.000000,0.000000,1000.000000") (Variables|Default|GetObject%20Types_0) false (Utilities|Array|MakeArray MyTarget) "None" true "(R=1.000000,G=0.000000,B=0.000000,A=1.000000)" "(R=0.000000,G=1.000000,B=0.000000,A=1.000000)" 9.0))
  (Transformation|SetWorldLocation (Utilities|Array|Get(acopy) ParticleComponents) (+ (.Location (Collision|BreakHitResult _outhit)) (* (Transformation|GetActorForwardVector MyTarget) 1e+02)) false false)
  (bind _asplayer_pawn_athena (Utilities|Casting|CastToPlayerPawn_Athena MyTarget))
  (Effects|Components|ParticleSystem|SetBoolParameter (Utilities|Array|Get(acopy) ParticleComponents) "IsLocalPlayer" (Pawn|IsLocallyControlled _asplayer_pawn_athena))
  (return))
