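; Parent gameplay cue for Sinister Melon melee combo hits that orients the burst particle effect to the player's swing.
; On burst, it turns the first particle component to face opposite the pawn's aim right vector and passes the impact direction as the "Direction" parameter.
; It then reads the weapon's Is Right flag to add a right or left swing pitch offset, writing world rotation values back as relative rotation.
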(fn OnBurst(EitherVFX-preferthisversion)
  (params (MyTarget :type Actor%20Object%20Reference) (Parameters :type Gameplay%20Cue%20Parameters%20Structure%20%28by%20ref%29) (ParticleComponents :type Array%20of%20FXSystem%20Component%20Object%20References) (AudioComponents :type Array%20of%20Audio%20Component%20Object%20References) (BurstCameraShakeInstance :type Camera%20Shake%20Base%20Object%20Reference) (BurstDecalInstance :type Decal%20Actor%20Object%20Reference))
  (bind _asfort_player_pawn_athena (Utilities|Casting|CastToFortPlayerPawnAthena MyTarget))
  (bind _output (Utilities|Array|Get(acopy) ParticleComponents))
  (Utilities|FlowControl|Sequence
    (:then_0
      (Transformation|SetWorldRotation _output (Math|Conversions|RotationFromXVector (* (Math|Vector|GetRightVector (Pawn|GetBaseAimRotation _asfort_player_pawn_athena)) -1.0)) false false)
      (Effects|Components|ParticleSystem|SetVectorParameter _output "Direction" (Variables|Default|GetImpactDirection)))
    (:then_1
      (bind _asb_weap_sinister_melon (Utilities|Casting|CastToB_Weap_SinisterMelon (Variables|/Script/FortniteGame.FortPawn|GetCurrentWeapon _asfort_player_pawn_athena)))
      (Transformation|SetRelativeRotation _output (Math|Rotator|SelectRotator (Math|Rotator|MakeRotator (.roll (Transformation|GetWorldRotation _output)) (+ (.pitch (Transformation|GetWorldRotation _output)) (Variables|Default|GetRightSwingOffset)) (.yaw (Transformation|GetWorldRotation _output))) (Math|Rotator|MakeRotator (.roll (Transformation|GetWorldRotation _output)) (+ (.pitch (Transformation|GetWorldRotation _output)) (Variables|Default|GetLeftSwingOffset)) (.yaw (Transformation|GetWorldRotation _output))) (Variables|/SinisterMelon/Gameplay/SinisterMelon/B_Weap_SinisterMelon.B_Weap_SinisterMelon_C|GetIs%20Right _asb_weap_sinister_melon)) false false))
    (:then_2
      (return))))
