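; Gameplay cue burst notify that plays impact effects for the SinisterMelon projectile.
; After the parent burst runs, it casts the target to Prj_SinisterMelon_Cross and spawns a short-lived, fading impact decal at that actor's transform.
; It then orients a spawned particle component from the projectile's up and right vectors and scales it uniformly from the Y scale of the projectile's Niagara component, so the impact follows the projectile's size.
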(fn OnBurst(EitherVFX-preferthisversion)
  (params (MyTarget :type Actor%20Object%20Reference) (Parameters :type Gameplay%20Cue%20Parameters%20Structure%20%28by%20ref%29) (ParticleComponents :type Array%20of%20FXSystem%20Component%20Object%20References) (AudioComponents :type Array%20of%20Audio%20Component%20Object%20References) (BurstCameraShakeInstance :type Camera%20Shake%20Base%20Object%20Reference) (BurstDecalInstance :type Decal%20Actor%20Object%20Reference))
  (bind _output (Utilities|Array|Get(acopy) ParticleComponents))
  (|CallParentFunction MyTarget Parameters ParticleComponents AudioComponents BurstCameraShakeInstance BurstDecalInstance)
  (bind _asprj_sinister_melon_cross (Utilities|Casting|CastToPrj_SinisterMelon_Cross MyTarget)
    (:then
      (bind _returnvalue (Rendering|Decal|SpawnDecalatLocation _asprj_sinister_melon_cross (Variables|Default|GetImpact_Decal) (Math|Vector|MakeVector 4e+02 4e+02 4e+02) (Transformation|GetActorLocation _asprj_sinister_melon_cross) (Transformation|GetActorRotation _asprj_sinister_melon_cross) 2.0))
      (Rendering|Components|Decal|SetFadeOut _returnvalue 0.0 2.0 true)
      (Transformation|SetWorldRotation _output (Math|Rotator|MakeRotfromXZ (Transformation|GetActorUpVector _asprj_sinister_melon_cross) (Transformation|GetActorRightVector _asprj_sinister_melon_cross)) false false)
      (Transformation|SetWorldScale3D _output (Math|Conversions|ToVector(Float) (/ (.y (Transformation|GetWorldScale (Variables|/SinisterMelon/Gameplay/SinisterMelon/Projectile/Prj_SinisterMelon_Cross.Prj_SinisterMelon_Cross_C|GetNS_SinisterMelon_Prj _asprj_sinister_melon_cross))) 1.35))))
    (:CastFailed
      _))
  (return))
